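# Unlit material for the lines of VR Editor floating text, as its name and folder indicate.
# A ColorAndOpacity vector parameter, white by default, supplies RGB to an EyeAdaptationInverse node at full alpha.
# That node cancels auto-exposure on the color, so its brightness holds steady as scene exposure changes.
# The parameter's alpha is unused despite its name.
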Material /Engine/VREditor/FloatingText/LineMaterial
graph {
 properties {pixelDepthOffsetMode=PDOM_Legacy, shadingModel=MSM_Unlit, usedShadingModels=Unlit}
 nodes [EyeAdaptationInverse(LightValueInput=VectorParameter() {parameterName=ColorAndOpacity, defaultValue=<rgba 1 1 1 1>}.RGB, AlphaInput=Constant() {r=1})]
}
